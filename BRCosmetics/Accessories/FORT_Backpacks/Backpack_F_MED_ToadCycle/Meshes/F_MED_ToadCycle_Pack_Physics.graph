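# Physics asset for the ToadCycle backpack cosmetic, set up for secondary motion on its lid.
# Three capsule bodies chain pack to base_lid to dyn_lid, with only dyn_lid carrying a sizable, offset collision capsule.
# The base_lid to pack joint leaves all angular axes free.
# The dyn_lid to base_lid joint acts like a narrow hinge, with a small swing range on one axis and zero limits on the others.
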PhysicsAsset /BRCosmetics/Accessories/FORT_Backpacks/Backpack_F_MED_ToadCycle/Meshes/F_MED_ToadCycle_Pack_Physics
bodies {
 base_lid [{shapeType=Capsule, length=0.5, radius=0.5}]
 dyn_lid [{shapeType=Capsule, length=2.5309, radius=5, shapeName=dyn_lid_capsule, center=<-2 0 -3>, rotation=<rot 0 0 -90>}]
 pack [{shapeType=Capsule, length=0, radius=0}]
}
constraints [
 {
  bone1Name=base_lid
  bone2Name=pack
  swing1LimitDegrees=45
  swing1Motion=Free
  swing2LimitDegrees=45
  swing2Motion=Free
  twistLimitDegrees=45
  twistMotion=Free
 }
 {bone1Name=dyn_lid, bone2Name=base_lid, swing1LimitDegrees=0, swing2LimitDegrees=18, twistLimitDegrees=0}
]
