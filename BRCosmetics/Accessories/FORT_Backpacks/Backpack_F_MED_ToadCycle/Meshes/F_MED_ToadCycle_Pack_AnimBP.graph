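# Animation Blueprint for the ToadCycle backpack that offsets its pack bone and adds rigid-body secondary motion.
# Event graph turns the owner's spine_05 yaw change, relative to spine_01, into clamped additive roll and Y translation for that bone.
# A RigidBody node then simulates the physics asset in pack-bone space with damped world motion.
# A parameter-driven gravity override is stored, though the node shows constant gravity.
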AnimBlueprint /BRCosmetics/Accessories/FORT_Backpacks/Backpack_F_MED_ToadCycle/Meshes/F_MED_ToadCycle_Pack_AnimBP
target_skeleton=&/F_MED_ToadCycle_Pack_Skeleton
graphs [
 AnimGraph {nodes [
   n1 = "Transform (Modify) Bone - Bone: pack"(
    ComponentPose=Local To Component(LocalPose=Input Pose())
    Translation=Get BP Translate() -> (BP Translate:Vector)
    Rotation=Get BP Rotate() -> (BP Rotate:Rotator)
   ) {
    rotationMode=BMM_Additive
    rotationSpace=BCS_BoneSpace
    translationMode=BMM_Additive
    translationSpace=BCS_BoneSpace
    alphaBoolBlend {blendInTime=2, blendOption=QuarticInOut, blendOutTime=5}
    alphaScaleBiasClamp {bInterpResult=true}
    boneToModify {boneName=pack}
   }
   n2 = RigidBody!(ComponentPose=@n1, Alpha, OverrideWorldGravity:Vector="0,0,-980", ExternalForce) -> (Pose) {
    bClampLinearTranslationLimitToRefPose=true
    bOverrideWorldGravity=true
    overridePhysicsAsset=&/F_MED_ToadCycle_Pack_Physics
    simulationSpace=BaseBoneSpace
    alphaScaleBiasClamp {interpSpeedDecreasing=5, interpSpeedIncreasing=2}
    baseBoneRef {boneName=pack}
    componentAppliedLinearAccClamp=<1500 1500 1500>
    simSpaceSettings {
     dampingAlpha=0
     maxAngularAcceleration=12
     maxAngularVelocity=4
     maxLinearAcceleration=400
     maxLinearVelocity=100
     velocityScaleZ=0.25
     worldAlpha=0.25
    }
   }
   Output Pose(Result=Component To Local(ComponentPose=@n2)) {layerGroup=None}
  ]}
 EventGraph {nodes [
   Event BlueprintUpdateAnimation()
   > n1 = Sequence()
   n2 = GetOwnerSkeletalMeshComponent()
   n3 = BreakRotator(InRot=BreakTransform(InTransform=GetDeltaTransformFromRefPose(self=@n2, BoneName=spine_05, BaseName=spine_01)).Rotation)
   n4 = MapRangeClamped(Value=@n3.Yaw, InRangeA=0, InRangeB=-20, OutRangeB=-15)
   Set BP Rotate(
    execute=@n1.then_1
    BP Rotate:Rotator=MakeRotator(Roll=FInterpEaseInOut!(A:double=0, B:double=@n4, Alpha:double=0, Exponent:double=0) -> (ReturnValue:double))
   ) -> (Output_Get:Rotator)
   n5 = MapRangeClamped(Value=@n3.Yaw, InRangeA=0, InRangeB=-10, OutRangeB=-2.5)
   > Set BP Translate(BP Translate:Vector=MakeVector(Y=FInterpEaseInOut(A=0, B=@n5, Alpha=0, Exponent=0))) -> (Output_Get:Vector)
   n6 = "Gravity Override State Function (Parameter Asset) (Nativized)"(
    execute=@n1.then_0
    CharacterPartAnimInstance=Self-Reference()
    GameTimeInSeconds=GetGameTimeInSeconds()
    Parameters=&/F_MED_ToadCycle_Pack_Parameters
   )
   > Set Gravity Override(Gravity Override=@n6)
  ]}
]
variables [
 BP Rotate {type=Rotator}
 BP Translate {type=Vector}
 Gravity Override {type=Vector}
]
